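# Physics asset for the Scuba pickaxe skeletal mesh, giving its dangling tanks, hoses and harpoon tether bodies for secondary motion.
# Capsule bodies cover the two-handed handle, tether and each hose, while the tanks bone carries a pair of capsules.
# Each dyn_ body is constrained directly to the handle in a star layout.
# The tanks lock both swing axes and move only in twist, while hoses and tether lock twist and move in swing.
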PhysicsAsset /BRCosmetics/Weapons/FORT_Melee/Pickaxe_Scuba/Meshes/SK_Scuba_Pickaxe_Physics
bodies {
 Melee_TwoHanded_Handle [{shapeType=Capsule, length=103.3161, radius=3.1856, center=<0.68 0.4346 26.2716>}]
 dyn_harpoon_tether [{shapeType=Capsule, length=5.1185, radius=2.284, center=<3.7194 -0.0295 -0.0003>, rotation=<rot -79.5437 89.542 -89.9972>}]
 dyn_hose_l [{shapeType=Capsule, length=6.3432, radius=1.926, center=<2.8281 1.5877 -5.3556>, rotation=<rot -4.0017 -90 -19.2341>}]
 dyn_hose_r [{shapeType=Capsule, length=5.8728, radius=1.9285, center=<-3.1225 1.6292 -5.8336>, rotation=<rot -0.6073 -90 17.1319>}]
 dyn_tanks [
  {shapeType=Capsule, length=24.3728, radius=5.7773, center=<7.6716 -2.5073 -11.6728>, rotation=<rot 0 0 8.4398>}
  {shapeType=Capsule, length=24.3728, radius=5.7773, center=<-5.3106 -2.5072 -11.6728>, rotation=<rot 0 0 8.4403>}
 ]
}
constraints [
 {
  bone1Name=dyn_tanks
  bone2Name=Melee_TwoHanded_Handle
  swing1LimitDegrees=45
  swing1Motion=Locked
  swing2LimitDegrees=45
  swing2Motion=Locked
  twistLimitDegrees=20
 }
 {
  bone1Name=dyn_hose_r
  bone2Name=Melee_TwoHanded_Handle
  swing1LimitDegrees=20
  swing2LimitDegrees=20
  twistLimitDegrees=45
  twistMotion=Locked
 }
 {
  bone1Name=dyn_hose_l
  bone2Name=Melee_TwoHanded_Handle
  swing1LimitDegrees=20
  swing2LimitDegrees=20
  twistLimitDegrees=45
  twistMotion=Locked
 }
 {
  bone1Name=dyn_harpoon_tether
  bone2Name=Melee_TwoHanded_Handle
  swing1LimitDegrees=10
  swing2LimitDegrees=70
  twistLimitDegrees=45
  twistMotion=Locked
 }
]
